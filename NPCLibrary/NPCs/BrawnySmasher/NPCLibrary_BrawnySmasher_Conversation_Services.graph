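# Conversation database for the BrawnySmasher NPC's services menu, offering item sales, upgrade, Rift-To-Go, prop disguise, storm forecast and disguise reveal.
# One entry point fans out to eight placeholder-text speech nodes, each gated by a HasService tag requirement and shown as a preview choice.
# Each branch leads to its service action or a Back node.
# The three sell slots play a give-item animation, then share one Repeat node.
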ConversationDatabase /NPCLibrary/NPCs/BrawnySmasher/NPCLibrary_BrawnySmasher_Conversation_Services
entry_points [{tag=Conversation.Entry.BrawnySmasher.Services, destinations [#1]}]
nodes [
 ConversationEntryPointNode_7 {
  id=#1
  class=ConversationEntryPointNode
  outputs [
   #2
   #3
   #4
   #5
   #6
   #7
   #8
   #9
  ]
  properties {entryTag {tagName=Conversation.Entry.BrawnySmasher.Services}}
 }
 Speech_C_9 {
  id=#9
  class=Speech_C
  outputs [#10, #11]
  properties {softTaskIcon=&T_UI_NPC_Intel2, generalConfig {defaultMessage="[PLACEHOLDER]"}}
  sub_nodes [
   FortConversationRequirement_HasService_9 {class=FortConversationRequirement_HasService, properties {serviceTag {tagName=Tandem.Service.RevealDisguise}}}
   FortConversationChoiceWithPreviewNode_9 {class=FortConversationChoiceWithPreviewNode, properties {defaultChoiceDisplayText=Reveal}}
  ]
 }
 RevealDisguise_C_0 {
  id=#11
  class=RevealDisguise_C
  outputs []
  properties {softTaskIcon=&T_UI_NPC_Intel2}
  sub_nodes [ConversationChoiceNode_13 {class=ConversationChoiceNode, properties {defaultChoiceDisplayText=CONFIRM}}]
 }
 Back_C_10 {
  id=#10
  class=Back_C
  sub_nodes [
   ConversationChoiceNode_16 {class=ConversationChoiceNode, properties {defaultChoiceDisplayText=BACK}}
   SideEffect_Client_PlayVOEvent_C_1 {class=SideEffect_Client_PlayVOEvent_C, properties {voiceOverTag {tagName=Pawn.Athena.NPC.Sound.Voice.Quest.Disappointed}}}
  ]
 }
 Speech_C_7 {
  id=#8
  class=Speech_C
  outputs [#12, #13]
  properties {softTaskIcon=/Game/Athena/HUD/Interaction/T_UI_NPC_Map, generalConfig {defaultMessage="[PLACEHOLDER]"}}
  sub_nodes [
   FortConversationRequirement_HasService_5 {class=FortConversationRequirement_HasService, properties {serviceTag {tagName=Tandem.Service.ShowFutureStormCircle}}}
   FortConversationChoiceWithPreviewNode_6 {class=FortConversationChoiceWithPreviewNode, properties {defaultChoiceDisplayText=STORM FORECAST}}
  ]
 }
 ShowFutureStormCircle_C_0 {
  id=#13
  class=ShowFutureStormCircle_C
  outputs []
  sub_nodes [ConversationChoiceNode_15 {class=ConversationChoiceNode, properties {defaultChoiceDisplayText=CONFIRM}}]
 }
 Back_C_8 {
  id=#12
  class=Back_C
  sub_nodes [ConversationChoiceNode_11 {class=ConversationChoiceNode, properties {defaultChoiceDisplayText=BACK}}]
 }
 Speech_C_6 {
  id=#7
  class=Speech_C
  outputs [#14, #15]
  properties {softTaskIcon=/Game/Athena/HUD/Interaction/T_UI_NPC_Intel, generalConfig {defaultMessage="[PLACEHOLDER]"}}
  sub_nodes [
   FortConversationRequirement_HasService_8 {class=FortConversationRequirement_HasService, properties {serviceTag {tagName=Tandem.Service.PropDisguise}}}
   FortConversationChoiceWithPreviewNode_8 {class=FortConversationChoiceWithPreviewNode, properties {defaultChoiceDisplayText=PROP DISGUISE}}
  ]
 }
 PropDisguise_C_0 {
  id=#15
  class=PropDisguise_C
  outputs []
  sub_nodes [ConversationChoiceNode_14 {class=ConversationChoiceNode, properties {defaultChoiceDisplayText=CONFIRM}}]
 }
 Back_C_6 {
  id=#14
  class=Back_C
  sub_nodes [ConversationChoiceNode_12 {class=ConversationChoiceNode, properties {defaultChoiceDisplayText=BACK}}]
 }
 Speech_C_1 {
  id=#6
  class=Speech_C
  outputs [#16, #17]
  properties {softTaskIcon=/Game/Athena/HUD/Interaction/T_UI_NPC_Rift, generalConfig {defaultMessage="[PLACEHOLDER]"}}
  sub_nodes [
   FortConversationRequirement_HasService_7 {class=FortConversationRequirement_HasService, properties {serviceTag {tagName=Tandem.Service.RiftToGo}}}
   FortConversationChoiceWithPreviewNode_7 {class=FortConversationChoiceWithPreviewNode, properties {defaultChoiceDisplayText=RIFT-TO-GO}}
  ]
 }
 RiftToGo_C_0 {
  id=#17
  class=RiftToGo_C
  outputs []
  properties {softTaskIcon=&T_UI_NPC_Rift}
  sub_nodes [ConversationChoiceNode_9 {class=ConversationChoiceNode, properties {defaultChoiceDisplayText=CONFIRM}}]
 }
 Back_C_4 {
  id=#16
  class=Back_C
  sub_nodes [ConversationChoiceNode_2 {class=ConversationChoiceNode, properties {defaultChoiceDisplayText=BACK}}]
 }
 Speech_C_3 {
  id=#5
  class=Speech_C
  outputs [#18, #19]
  properties {softTaskIcon=/Game/Athena/HUD/Interaction/T_UI_NPC_Money, generalConfig {defaultMessage="[PLACEHOLDER]"}}
  sub_nodes [
   FortConversationRequirement_HasService_4 {class=FortConversationRequirement_HasService, properties {serviceTag {tagName=Tandem.Service.Sell3}}}
   FortConversationChoiceWithPreviewNode_3 {class=FortConversationChoiceWithPreviewNode, properties {defaultChoiceDisplayText=BUY}}
  ]
 }
 SellLoot_C_2 {
  id=#19
  class=SellLoot_C
  outputs [#20]
  properties {sellSlot=3}
  sub_nodes [SideEffect_Client_PlayAnimationEvent_C_2 {class=SideEffect_Client_PlayAnimationEvent_C, properties {animationEventName=aie.npc.GiveItem}}]
 }
 Repeat_C_1 {id=#20, class=Repeat_C}
 Back_C_2 {
  id=#18
  class=Back_C
  sub_nodes [ConversationChoiceNode_3 {class=ConversationChoiceNode, properties {defaultChoiceDisplayText=BACK}}]
 }
 Speech_C_2 {
  id=#4
  class=Speech_C
  outputs [#21, #22]
  properties {softTaskIcon=&T_UI_NPC_Money, generalConfig {defaultMessage="[PLACEHOLDER]"}}
  sub_nodes [
   FortConversationRequirement_HasService_2 {class=FortConversationRequirement_HasService, properties {serviceTag {tagName=Tandem.Service.Sell2}}}
   FortConversationChoiceWithPreviewNode_1 {class=FortConversationChoiceWithPreviewNode, properties {defaultChoiceDisplayText=BUY}}
  ]
 }
 SellLoot_C_1 {
  id=#22
  class=SellLoot_C
  outputs [#20]
  properties {sellSlot=2}
  sub_nodes [SideEffect_Client_PlayAnimationEvent_C_0 {class=SideEffect_Client_PlayAnimationEvent_C, properties {animationEventName=aie.npc.GiveItem}}]
 }
 Back_C_1 {
  id=#21
  class=Back_C
  sub_nodes [ConversationChoiceNode_0 {class=ConversationChoiceNode, properties {defaultChoiceDisplayText=BACK}}]
 }
 Speech_C_15 {
  id=#3
  class=Speech_C
  outputs [#23, #24]
  properties {softTaskIcon=/Game/Athena/HUD/Interaction/T_UI_NPC_Upgrade, generalConfig {defaultMessage="[PLACEHOLDER]"}}
  sub_nodes [
   FortConversationRequirement_HasService_1 {class=FortConversationRequirement_HasService, properties {serviceTag {tagName=Tandem.Service.Upgrade}}}
   FortConversationChoiceWithPreviewNode_5 {class=FortConversationChoiceWithPreviewNode, properties {defaultChoiceDisplayText=UPGRADE}}
  ]
 }
 UpgradeItem_C_0 {
  id=#24
  class=UpgradeItem_C
  outputs []
  sub_nodes [ConversationChoiceNode_8 {class=ConversationChoiceNode, properties {defaultChoiceDisplayText=CONFIRM}}]
 }
 Back_C_0 {
  id=#23
  class=Back_C
  sub_nodes [ConversationChoiceNode_6 {class=ConversationChoiceNode, properties {defaultChoiceDisplayText=BACK}}]
 }
 Speech_C_0 {
  id=#2
  class=Speech_C
  outputs [#25, #26]
  properties {softTaskIcon=&T_UI_NPC_Money, generalConfig {defaultMessage="[PLACEHOLDER]"}}
  sub_nodes [
   FortConversationRequirement_HasService_3 {class=FortConversationRequirement_HasService, properties {serviceTag {tagName=Tandem.Service.Sell1}}}
   FortConversationChoiceWithPreviewNode_2 {class=FortConversationChoiceWithPreviewNode, properties {defaultChoiceDisplayText=BUY}}
  ]
 }
 SellLoot_C_3 {
  id=#26
  class=SellLoot_C
  outputs [#20]
  sub_nodes [SideEffect_Client_PlayAnimationEvent_C_4 {class=SideEffect_Client_PlayAnimationEvent_C, properties {animationEventName=aie.npc.GiveItem}}]
 }
 Back_C_7 {
  id=#25
  class=Back_C
  sub_nodes [ConversationChoiceNode_10 {class=ConversationChoiceNode, properties {defaultChoiceDisplayText=BACK}}]
 }
]
