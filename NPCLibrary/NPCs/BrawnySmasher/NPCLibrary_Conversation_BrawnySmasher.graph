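# Conversation database for the Brawny Smasher NPC, a disguised character posing as human with comedic dialogue.
# Its entry point leads to a greeting, with a special line for the first-ever conversation, branching to the shared BasicServices and Season17 Brawny conversations.
# A third branch, gated on the RevealDisguise service, steps through choices until "Reveal" triggers the RevealDisguise task.
# Back-out choices along that branch use ReturnToStart nodes.
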ConversationDatabase /NPCLibrary/NPCs/BrawnySmasher/NPCLibrary_Conversation_BrawnySmasher
entry_points [{tag=Conversation.Entry.BrawnySmasher, destinations [#1]}]
nodes [
 ConversationEntryPointNode_7 {
  id=#1
  class=ConversationEntryPointNode
  outputs [#2]
  properties {entryTag {tagName=Conversation.Entry.BrawnySmasher}}
 }
 Speech_C_4 {
  id=#2
  class=Speech_C
  outputs [#3, #4, #5]
  properties {
   nodeName=""
   generalConfig {
    defaultMessage="How pleasing it is to converse with a <b>fellow human!</>"
    contextualMessages [{
      message="You and I have <b>so much</> in common! \r\nHuman skin, human face, human emotions, etcetera! "
      contextRequirements [{requirement {class=FortControllerRequirement_IsFirstEverConversationWithNPC}}]
      priority {curve {rowName=Tandem.Speech.Priority.FirstEverInteraction}}
     }]
   }
  }
 }
 ConversationLinkNode_0 {id=#5, class=ConversationLinkNode, properties {remoteEntryTag {tagName=Conversation.Entry.BasicServices}}}
 Speech_C_0 {
  id=#4
  class=Speech_C
  outputs [#6, #7]
  properties {
   nodeName=""
   generalConfig {defaultMessage="I simply require <b>nutrients</>.\r\nI am... what is the human word for it... <enemy>hangry? </>"}
  }
  sub_nodes [
   ConversationChoiceNode_1 {class=ConversationChoiceNode, properties {defaultChoiceDisplayText=Uhh... are you okay?}}
   FortConversationRequirement_HasService_0 {class=FortConversationRequirement_HasService, properties {serviceTag {tagName=Tandem.Service.RevealDisguise}}}
  ]
 }
 Speech_C_2 {
  id=#7
  class=Speech_C
  outputs [#8, #9]
  properties {nodeName="", generalConfig {defaultMessage="Tuna-Flavored Kitten Food!"}}
  sub_nodes [ConversationChoiceNode_0 {class=ConversationChoiceNode, properties {defaultChoiceDisplayText=What's your favourite food?}}]
 }
 Speech_C_9 {
  id=#9
  class=Speech_C
  outputs [#10]
  properties {softTaskIcon=&T_UI_NPC_Intel2, generalConfig {defaultMessage=...I think I just blew my cover.}}
  sub_nodes [FortConversationChoiceWithPreviewNode_9 {class=FortConversationChoiceWithPreviewNode, properties {defaultChoiceDisplayText=Reveal}}]
 }
 RevealDisguise_C_0 {
  id=#10
  class=RevealDisguise_C
  outputs []
  properties {softTaskIcon=&T_UI_NPC_Intel2}
 }
 ReturnToStart_C_1 {
  id=#8
  class=ReturnToStart_C
  properties {softTaskIcon=&T_UI_NPC_Back}
  sub_nodes [ConversationChoiceNode_5 {class=ConversationChoiceNode, properties {defaultChoiceDisplayText=BACK}}]
 }
 ReturnToStart_C_0 {
  id=#6
  class=ReturnToStart_C
  properties {softTaskIcon=&T_UI_NPC_Back}
  sub_nodes [ConversationChoiceNode_4 {class=ConversationChoiceNode, properties {defaultChoiceDisplayText=Whatever...}}]
 }
 ConversationLinkNode_1 {id=#3, class=ConversationLinkNode, properties {remoteEntryTag {tagName=Conversation.Entry.Season17.Brawny}}}
]
